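# Parent material for the CornerWeek pickaxe, adding animated glowing branch, fiber and vein effects to the shared CharacterParent 2023 PreFX/PostFX functions.
# Noise-distorted FX mask textures, rotated swirl layers with BumpOffset depth, panning light spots and shavings, and a bounding-box-driven vein pulse build the effect layers.
# These feed overrides of emissive, base color, specular, height-derived normal and masked noise vertex offset.
# A static switch shows the Tech Art mask for debugging.
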Material /BRCosmetics/Weapons/FORT_Melee/Pickaxe_CornerWeek/Materials/M_Pickaxe_CornerWeek_Parent
graph {
 properties {
  bUseMaterialAttributes=true
  bUsedWithMeshDeformer=true
  bUsedWithMorphTargets=true
  bUsedWithSkeletalMesh=true
  pixelDepthOffsetMode=PDOM_Legacy
  shadingModel=MSM_FromMaterialExpression
  usedShadingModels=Default Lit | Subsurface | Clear Coat | Subsurface Profile | Cloth
 }
 nodes [
  MaterialFunctionCall() {materialFunction=&ObjectPivotPoint}
  n1 = TextureCoordinate()
  n2 = ScalarParameter() {defaultValue=9, group=Corner Week, parameterName=Branches - Tiling}
  n3 = VectorParameter() {group=Corner Week, parameterName=Branch - Speed, defaultValue=<rgba -0.02 0.0728 0 0>}
  n4 = TextureSample() {samplerSource=SSM_Wrap_WorldGroupSettings, samplerType=SAMPLERTYPE_Grayscale, texture=&LowResBlurredNoise}
  n5 = Add(
   A=Multiply(A=Time(), B=ComponentMask(None=@n3.RGB) {g=true, r=true})
   B=Multiply(A=@n4.R, B=ScalarParameter() {defaultValue=0.3, group=Corner Week, parameterName=Branches - Distortion Amount})
  )
  n6 = ScalarParameter() {defaultValue=-0.0632, parameterName=Param_1}
  n7 = MaterialFunctionCall(
   UVs=Add(A=Multiply(A=@n1, B=@n2), B=NamedRerouteDeclaration(None=@n5) {name=Distorted UVs, nodeColor=<rgba 1 0.1172 0 1>})
   "Rotation Angle (0-1)"=@n6
  ) {materialFunction=&CustomRotator}
  n8 = Panner(
   Coordinate=@n7.Rotated Values
   Speed=VectorParameter() {group=Corner Week, parameterName=Branch 02 - Panning, defaultValue=<rgba 0.02 -0.1456 0 0>}.RGB
  )
  MaterialFunctionCall(UVs=@n8, "Rotation Angle (0-1)"=@n6) {materialFunction=&CustomRotator}
  n9 = VectorParameter() {group=Corner Week, parameterName=Light - Direction, defaultValue=<rgba -1 1 1 0>}
  n10 = PixelNormalWS()
  Multiply(A=Multiply(A=NamedRerouteUsage(), B=DotProduct(A=@n9.RGB, B=@n10))) {constB=0}
  n11 = TextureCoordinate()
  n12 = ScalarParameter() {defaultValue=11.5523, group=Corner Week, parameterName=Global - Distortion Tiling}
  n13 = TextureSample(UVs=Panner(
    Coordinate=Multiply(A=@n11, B=@n12)
    Speed=VectorParameter() {group=Corner Week, parameterName=Global - Distortion Speed, defaultValue=<rgba 0.312 0 0 0>}.RGB
   )) {samplerSource=SSM_Wrap_WorldGroupSettings, samplerType=SAMPLERTYPE_Grayscale, texture=&LowResBlurredNoise}
  n14 = Add(
   A=Multiply(A=@n13.R, B=ScalarParameter() {defaultValue=0.005, group=Corner Week, parameterName=Global - Distortion Amount})
   B=TextureCoordinate()
  )
  n15 = TextureSample(UVs=@n14) {texture=/BRCosmetics/Weapons/FORT_Melee/Pickaxe_CornerWeek/Textures/T_CornerWeek_Pickaxe_FX2}
  n16 = NamedRerouteUsage()
  NamedRerouteDeclaration(None=Max(A=@n15.R, B=Multiply(A=@n15.R, B=@n16))) {name=Fibers, nodeColor=<rgba 0.3594 1 0 1>}
  n17 = MaterialFunctionCall() {materialFunction=&MF_CharacterParent_2023_PreFX}
  n18 = GetMaterialAttributes(None=MaterialFunctionCall(Material=@n17.Result, M=@n17.Result, Pre Skinned Local Position=@n17.Result, Distance Blend=@n17.Result, Fresnel=@n17.Result) {materialFunction=&MF_CharacterParent_2023_PostFX}.Result) {attributeGetTypes [
    #1
    #2
    #3
    #4
   ]}
  NamedRerouteDeclaration(None=@n18.Emissive Color) {name=Emissive In, nodeColor=<rgba 0 1 0.6016 1>}
  NamedRerouteDeclaration(None=@n15.A) {name=Hard Edge Mask, nodeColor=<rgba 0 1 0.9063 1>}
  n19 = Multiply(
   A=MaterialFunctionCall() {materialFunction=&BoundingBoxBased_0-1_UVW}.B
   B=ScalarParameter() {defaultValue=5.5169, group=Corner Week, parameterName=Vein Pulse - Tiling}
  )
  n20 = Multiply(A=Time(), B=ScalarParameter() {defaultValue=-0.496, group=Corner Week, parameterName=Vein Pulse - Speed})
  n21 = Divide(A=Frac(None=Add(A=@n19, B=@n20)), B=ScalarParameter() {defaultValue=0.6359, group=Corner Week, parameterName=Vein Pulse - Length}) {constB=0.2}
  n22 = ConstantBiasScale(None=OneMinus(None=Saturate(None=@n21))) {bias=-0.5, scale=2}
  n23 = ScalarParameter() {group=Corner Week, parameterName=Vein Pulse - Bias}
  n24 = SmoothStep(Min=@n23, Value=@n22)
  n25 = SmoothStep(Min=@n23, Value=@n22) {constMax=-1}
  n26 = Power(
   Base=OneMinus(None=If(A=@n22, B=@n23, "A > B"=@n24, "A == B"=@n24, "A < B"=@n25))
   Exp=ScalarParameter() {defaultValue=2, group=Corner Week, parameterName=Vein Pulse - Curve Exp}
  )
  n27 = NamedRerouteUsage()
  NamedRerouteDeclaration(None=Multiply(A=Multiply(A=@n26, B=@n27), B=ScalarParameter() {defaultValue=100, group=Corner Week, parameterName=Vein Pulse - Strength})) {name=Vein Pulse, nodeColor=<rgba 1 0.1406 0 1>}
  NamedRerouteDeclaration(None=@n18.Base Color) {name=Base Color In, nodeColor=<rgba 0.4766 0 1 1>}
  NamedRerouteDeclaration(None=@n15.B) {name=Vein Mask, nodeColor=<rgba 0 1 0.6953 1>}
  NamedRerouteDeclaration(None=SetMaterialAttributes!(MaterialAttributes, Base Color=NamedRerouteDeclaration(None=@n17.Tech Art Mask) {name=TA Mask, nodeColor=<rgba 0 0.8359 1 1>}) {attributeSetTypes [#2]}) {name=Debug Material, nodeColor=<rgba 1 0 0.2812 1>}
  n28 = VectorParameter() {group=Corner Week, parameterName=Void Colour, defaultValue=<rgba 10 0.1915 0.0746 1>}
  n29 = TextureCoordinate()
  n30 = ScalarParameter() {defaultValue=4.2516, group=Corner Week, parameterName=Light Spots - Tiling}
  n31 = TextureSample(UVs=Panner(
    Coordinate=Multiply(A=@n29, B=@n30)
    Speed=VectorParameter() {group=Corner Week, parameterName=Light Spots - Speed, defaultValue=<rgba 0.05 0.1 0 0>}.RGB
   )) {samplerSource=SSM_Wrap_WorldGroupSettings, samplerType=SAMPLERTYPE_Grayscale, texture=&LowResBlurredNoise}
  n32 = Power(Base=@n31.R, Exp=ScalarParameter() {defaultValue=2.7275, group=Corner Week, parameterName=Light Spots - Exp})
  NamedRerouteDeclaration(None=Multiply(A=Multiply(A=@n28.RGB, B=@n32), B=ScalarParameter() {defaultValue=5.4644, group=Corner Week, parameterName=Light Spots - Intensity})) {name=Light Spots, nodeColor=<rgba 0 0.6016 1 1>}
  n33 = Power(Base=NamedRerouteUsage(), Exp=ScalarParameter() {defaultValue=0.5, group=Corner Week, parameterName=Fiber - Exp})
  n34 = ScalarParameter() {defaultValue=2, group=Corner Week, parameterName=Fiber - Intensity}
  n35 = LinearInterpolate(
   A=LinearInterpolate(
    A=NamedRerouteUsage()
    B=VectorParameter() {group=Corner Week, parameterName=Void Edge - Colour, defaultValue=<rgba 0.6308 0.0168 0.0033 1>}.RGB
    Alpha=NamedRerouteUsage()
   )
   B=VectorParameter() {group=Corner Week, parameterName=Backlit - Color, defaultValue=<rgba 10 0.1915 0.0746 1>}.RGB
   Alpha=Saturate(None=Multiply(A=@n33, B=@n34))
  )
  n36 = OneMinus(None=Saturate(None=NamedRerouteUsage()))
  n37 = Constant()
  n38 = SmoothStep(Min=@n37, Max=Add(A=@n37, B=Constant() {r=0.232}), Value=NamedRerouteUsage())
  n39 = Constant() {r=0.6808}
  n40 = Add(A=@n39, B=Constant() {r=0.0448})
  n41 = MaterialFunctionCall() {materialFunction=&BoundingBoxBased_0-1_UVW}
  n42 = LinearInterpolate(
   B=ScalarParameter() {defaultValue=0.3, group=Corner Week, parameterName=Branch - Opacity}
   Alpha=Multiply(A=Saturate(None=@n38), B=SmoothStep(Min=@n39, Max=@n40, Value=@n41.B))
  ) {constA=1}
  n43 = Multiply(A=NamedRerouteUsage(), B=ScalarParameter() {defaultValue=13.5754, group=Corner Week, parameterName=Branch - Intensity})
  n44 = Multiply(A=Add(A=NamedRerouteUsage(), B=@n35), B=Multiply(A=Max(A=@n36, B=@n42), B=@n43))
  n45 = Subtract(A=NamedRerouteUsage(), B=@n43)
  n46 = NamedRerouteUsage()
  n47 = LinearInterpolate(
   A=NamedRerouteUsage()
   B=Multiply(
    A=ScalarParameter() {defaultValue=0.75, group=Corner Week, parameterName=Backlit - Strength}
    B=LinearInterpolate(A=@n44, B=@n35, Alpha=@n45)
   )
   Alpha=Saturate(None=Multiply(A=@n46) {constB=200})
  )
  n48 = NamedRerouteUsage()
  NamedRerouteDeclaration(None=LinearInterpolate(
    A=@n47
    B=MaterialFunctionCall(Base=@n47, Blend=@n48) {materialFunction=&Blend_Overlay}.Result
    Alpha=NamedRerouteUsage()
   )) {name=Emissive, nodeColor=<rgba 0.9219 1 0 1>}
  NamedRerouteDeclaration(None=Multiply(
    A=TextureSample(UVs=@n14) {samplerType=SAMPLERTYPE_Masks, texture=/BRCosmetics/Weapons/FORT_Melee/Pickaxe_CornerWeek/Textures/T_CornerWeek_Pickaxe_FX}.A
    B=Constant() {r=0.75}
   )) {name=Void Edge, nodeColor=<rgba 0.1484 0 1 1>}
  n49 = ScalarParameter() {defaultValue=-0.1352, group=Corner Week, parameterName=Branch 01 - Step}
  n50 = ScalarParameter() {defaultValue=1.2, group=Corner Week, parameterName=Branch 01 - Smoothness}
  n51 = TextureSample(UVs=BumpOffset(Coordinate=@n7.Rotated Values, Height=ScalarParameter() {defaultValue=-3, group=Corner Week, parameterName=Branch 01 - Depth})) {
   samplerSource=SSM_Wrap_WorldGroupSettings
   texture=/BRCosmetics/Weapons/FORT_Melee/Pickaxe_BentBaton/Materials/Swirly_Wisps_Vertical_Dissolve
  }
  n52 = Multiply(
   A=SmoothStep(Min=@n49, Max=Add(A=@n49, B=@n50), Value=@n51.R)
   B=ScalarParameter() {defaultValue=1, group=Corner Week, parameterName=Branch 01 - Intensity}
  )
  n53 = VectorParameter() {group=Corner Week, parameterName=Branch 02 - UV Offset, defaultValue=<rgba 0.224 0.264 0 0>}
  n54 = TextureSample(UVs=Add(A=@n8, B=ComponentMask(None=@n53.RGB) {g=true, r=true})) {
   samplerSource=SSM_Wrap_WorldGroupSettings
   texture=&Swirly_Wisps_Vertical_Dissolve
  }
  n55 = Multiply(A=@n54.R, B=ScalarParameter() {defaultValue=1, group=Corner Week, parameterName=Branch 02 - Intensity})
  NamedRerouteDeclaration(None=Saturate(None=Max(A=@n52, B=@n55))) {name=Swirl, nodeColor=<rgba 1 0 0.0937 1>}
  NamedRerouteDeclaration(None=Multiply(A=NamedRerouteUsage(), B=Constant() {r=10})) {name=Sphere Mask, nodeColor=<rgba 1 0.9844 0 1>}
  n56 = TextureCoordinate()
  n57 = ScalarParameter() {defaultValue=6.1227, group=Corner Week, parameterName=Shavings - Tiling}
  n58 = TextureSample(UVs=Panner(
    Coordinate=Multiply(A=@n56, B=@n57)
    Speed=VectorParameter() {group=Corner Week, parameterName=Shavings - Speed, defaultValue=<rgba 0 0.0584 0 0>}.RGB
   )) {samplerType=SAMPLERTYPE_Grayscale, texture=/BRCosmetics/Weapons/FORT_Melee/Pickaxe_CornerWeek/Textures/T_GrungeShavings_01}
  n59 = OneMinus(None=NamedRerouteUsage())
  NamedRerouteDeclaration(None=Multiply(A=Multiply(A=@n58.R, B=@n59), B=ScalarParameter() {defaultValue=4.2375, group=Corner Week, parameterName=Shavings - Intensity})) {name=Shavings, nodeColor=<rgba 0.8047 1 0 1>}
  NamedRerouteDeclaration(None=@n15.G) {name=Branch Area, nodeColor=<rgba 0.3125 0 1 1>}
  NamedRerouteDeclaration(None=@n18.Normal) {name=Norm In, nodeColor=<rgba 1 0.2812 0 1>}
  NamedRerouteDeclaration(None=@n18.Specular) {name=Spec In, nodeColor=<rgba 1 0.0234 0 1>}
  OneMinus(None=@n46)
  SphereMask(
   A=Transform(None=WorldPosition().XYZ) {transformSourceType=TRANSFORMSOURCE_World, transformType=TRANSFORM_Local}
   B=Add(A=ObjectPositionWS(), B=VectorParameter() {defaultValue=<rgba 0 3.7928 38.0779 1184.626>}.RGB)
   Radius=ScalarParameter() {defaultValue=10.8413}
   Hardness=Constant()
  )
  n60 = NamedRerouteUsage()
  n61 = ScalarParameter() {defaultValue=2.3735, group=Corner Week, parameterName=Branch - Color Influence}
  n62 = Saturate(None=Multiply(A=NamedRerouteUsage(), B=Multiply(A=@n60, B=@n61)))
  n63 = MaterialFunctionCall!(Height=NamedRerouteUsage(), Absolute World Position, World Space Vertex Normals) {materialFunction=/Engine/Functions/Engine_MaterialFunctions03/Procedurals/HeightToNormalSmooth}
  n64 = Multiply(A=NamedRerouteUsage(), B=ScalarParameter() {defaultValue=8.7646, group=Corner Week, parameterName=Fiber - Normal Influence})
  n65 = Multiply(A=@n64, B=NamedRerouteUsage())
  n66 = TextureCoordinate()
  n67 = ScalarParameter() {defaultValue=55, group=Corner Week, parameterName=WPO - Tiling}
  n68 = TextureSample(UVs=Panner(
    Coordinate=Multiply(A=@n66, B=@n67)
    Speed=VectorParameter() {group=Corner Week, parameterName=WPO - Speed, defaultValue=<rgba 0.256 0.144 0 0>}.RGB
   )) {samplerSource=SSM_Wrap_WorldGroupSettings, samplerType=SAMPLERTYPE_Grayscale, texture=&LowResBlurredNoise}
  n69 = Multiply(A=VertexNormalWS(), B=Multiply(A=@n68.R, B=ScalarParameter() {defaultValue=0.6, group=Corner Week, parameterName=WPO - Offset Amount}))
  n70 = NamedRerouteUsage()
  n71 = ComponentMask(None=NamedRerouteUsage()) {b=true}
  n72 = SetMaterialAttributes!(
   MaterialAttributes=@n18.MaterialAttributes
   Emissive Color=LinearInterpolate(A=NamedRerouteUsage(), Alpha=@n62) {constB=0}
   Base Color=LinearInterpolate(
    A=NamedRerouteUsage()
    B=VectorParameter() {group=Corner Week, parameterName=Branch - Color, defaultValue=<rgba 0.0999 0.0561 0.0595 1>}.RGB
    Alpha=@n62
   )
   Normal=LinearInterpolate(
    A=NamedRerouteUsage()
    B=Transform(None=@n63.WorldSpaceNormal) {transformSourceType=TRANSFORMSOURCE_World, transformType=TRANSFORM_Tangent}
    Alpha=@n64
   )
   Specular=LinearInterpolate(A=NamedRerouteUsage(), Alpha=Saturate(None=@n65))
   World Position Offset=LinearInterpolate(A=Multiply(A=@n69, B=@n70), Alpha=Ceil(None=@n71)) {constB=0}
  ) {attributeSetTypes [
    #1
    #2
    #3
    #4
    #5
   ]}
  StaticSwitchParameter(True=NamedRerouteUsage(), False=@n72) {group=Corner Week - Debug, parameterName=Debug TA Mask}
  VertexInterpolator(VS=LocalPosition() {localOrigin=InstancePreSkinning}.XYZ)
 ]
}
