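# Unlit translucent Niagara material for the CornerWeek pickaxe's glowing crack effect.
# Emissive blends two colors through a center mask, modulated by a panning flow texture and particle color.
# Opacity selects crack, noise-warped stepped or emissive masks via scalar blends, then applies dynamic-parameter HeightLerp dissolve, sphere mask and depth fade.
# Optional SplineThicken offset takes its width from a dynamic parameter.
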Material /BRCosmetics/Weapons/FORT_Melee/Pickaxe_CornerWeek/Meshes/FX/M_Pickaxe_CornerWeek_Crack
graph {
 properties {
  bAutomaticallySetUsageInEditor=false
  bUsedWithNiagaraMeshParticles=true
  bUsedWithNiagaraRibbons=true
  bUsedWithNiagaraSprites=true
  blendMode=BLEND_Translucent
  shadingModel=MSM_Unlit
  usedShadingModels=Unlit
 }
 nodes [
  n1 = Multiply(A=MaterialFunctionCall!(Center Point World Position) {materialFunction=/Engine/Functions/Engine_MaterialFunctions02/UVs/WorldSpaceAlignedScreenCoordinates}.Offset and Scaled UVS) {constB=10}
  n2 = TextureCoordinate()
  n3 = VectorParameter() {parameterName=EmissiveTex, defaultValue=<rgba 1.5 3 0 1>}
  n4 = AppendVector(A=@n3.R, B=@n3.R)
  n5 = TextureSampleParameter2D(UVs=Panner(Coordinate=Multiply(A=@n2, B=@n4), Speed=AppendVector(A=@n3.B, B=@n3.B)) {speedX=0.1, speedY=0.1}) {
   parameterName=EmissiveTex
   samplerType=SAMPLERTYPE_LinearColor
   texture=/Game/Effects/Fort_Effects/Creatures/Elemental/Fire/MergedFireFlowMap
  }
  n6 = LinearInterpolate(A=ScalarParameter() {defaultValue=-0.4, parameterName=EmissiveTex_LerpA}, Alpha=@n5.RGB) {constA=-0.4}
  Add(A=Multiply(A=TextureSample(UVs=@n1) {texture=/Game/Global/Textures/Default/Tileable_Patterns/T_Thick_Smoke_05}.RGB) {constB=0.2}, B=@n6)
  n7 = ParticleColor()
  n8 = ChannelMaskParameter(None=TextureSample() {texture=&/T_Pickaxe_CornerWeek_Crack}.RGB) {parameterName=MaskTex_RGBChange}
  n9 = Add(A=ComponentMask(None=TextureCoordinate() {uTiling=2}) {r=true}) {constB=0}
  n10 = TextureSampleParameter2D(UVs=Panner(Coordinate=TextureCoordinate() {uTiling=2}, Speed=Constant2Vector() {g=1.3}) {speedY=1}) {parameterName=NoiseTex, samplerType=SAMPLERTYPE_LinearColor, texture=&T_NoisePack_LinearColor}
  n11 = Add(A=ComponentMask(None=TextureCoordinate()) {g=true}, B=Multiply(A=LinearInterpolate(Alpha=@n10.R) {constA=-2, constB=0.5}) {constB=0.3})
  n12 = TextureSample(UVs=Add(A=AppendVector(A=@n9, B=@n11), B=Constant2Vector() {g=0.2, r=-0.5}) {constB=0}) {samplerSource=SSM_Clamp_WorldGroupSettings, texture=/Game/Effects/Fort_Effects/Textures/Mask/T_HalfSphere_D}
  n13 = Floor(None=Multiply(A=@n12.R) {constB=15})
  n14 = Power(Base=Multiply(A=@n12.R) {constB=5})
  n15 = Multiply(A=@n8.R, B=Saturate(None=LinearInterpolate(A=@n13, B=@n14)))
  n16 = LinearInterpolate(
   A=LinearInterpolate(A=Multiply(A=@n7.A, B=@n8.R), B=Multiply(A=@n7.A, B=@n15), Alpha=ScalarParameter() {parameterName=Use_MaskChange})
   B=Multiply(A=@n6, B=@n7.A)
   Alpha=ScalarParameter() {parameterName=Use_EmissiveMask}
  )
  n17 = DynamicParameter() {paramNames [SplineWidth,Dissolve,Param3,Param4]}
  n18 = TextureSample(UVs=MaterialFunctionCall(XY=TextureCoordinate() {uTiling=1.1, vTiling=0.5}) {materialFunction=&Swizzle}.YX) {samplerType=SAMPLERTYPE_LinearGrayscale, texture=/Game/Athena/Prototype/Terrain/T_CloudCoverage}
  n19 = Saturate(None=Multiply(A=MaterialFunctionCall(Transition Phase=@n17.Dissolve, Height Texture=@n18.R) {materialFunction=&HeightLerp}.Alpha) {constB=3})
  n20 = TextureCoordinate()
  n21 = Constant() {r=0.5}
  n22 = ScalarParameter() {defaultValue=0.5, parameterName=SphereMask_Radius}
  DepthFade(
   Opacity=Multiply(A=Multiply(A=@n16, B=@n19), B=SphereMask(A=@n20, B=@n21, Radius=@n22) {attenuationRadius=0.5, hardnessPercent=0.5})
   FadeDistance=ScalarParameter() {defaultValue=5, parameterName=Depth Fade}
  )
  n23 = Multiply(
   A=ScalarParameter() {defaultValue=40, parameterName=SplineWidth}
   B=DynamicParameter() {paramNames [SplineWidth,Dissolve,Param3,Param4]}.SplineWidth
  )
  LinearInterpolate(
   A=Constant()
   B=MaterialFunctionCall(WidthBase=@n23, WidthTip=@n23, Expand U or V UV Channel=StaticBool()) {materialFunction=&SplineThicken}.WorldPositionOffset
   Alpha=ScalarParameter() {parameterName=Use_SplineThicken}
  )
  n24 = VectorParameter() {parameterName=ColorA, defaultValue=<rgba 0.0833 0.0043 0.0043 1>}
  n25 = VectorParameter() {parameterName=ColorB, defaultValue=<rgba 5 0.59 0.1 1>}
  n26 = Abs(None=Subtract(A=TextureCoordinate()) {constB=0.5})
  n27 = ComponentMask(None=OneMinus(None=Multiply(A=@n26) {constB=2})) {r=true}
  n28 = ScalarParameter() {defaultValue=8, parameterName=ColorLinearMaskPower}
  n29 = Multiply(
   A=Saturate(None=Power(Base=@n27, Exp=@n28) {constExponent=8})
   B=SphereMask(A=TextureCoordinate(), B=Constant() {r=0.5}, Radius=ScalarParameter() {defaultValue=0.25, parameterName=ColorSphereMaskRadius}) {attenuationRadius=0.25, hardnessPercent=0}
  )
  MaterialFunctionCall(Emissive=Multiply(A=LinearInterpolate(A=@n24.RGB, B=@n25.RGB, Alpha=@n29), B=Multiply(A=@n6, B=@n7.RGB))) {materialFunction=&MF_FX_EmissiveAdjustments}
  n30 = Add(A=TextureCoordinate(), B=ParticleRandom())
  n31 = VectorParameter() {parameterName=MaskTex, defaultValue=<rgba 1 1 0 0>}
  n32 = AppendVector(A=@n31.R, B=@n31.R)
  n33 = TextureSampleParameter2D(UVs=Panner(Coordinate=TextureCoordinate() {vTiling=0.5}) {speedX=-0.1, speedY=0.1}) {
   group=Distortion
   parameterName=NoiseTexture
   samplerType=SAMPLERTYPE_Normal
   sortPriority=0
   texture=/Game/Global/Textures/Default/Tileable_Patterns/T_StormCloudDetail01_N
  }
  n34 = ComponentMask(None=@n33.RGB) {g=true, r=true}
  n35 = ScalarParameter() {parameterName=Noise_intensity}
  n36 = Panner(Coordinate=Add(A=Multiply(A=@n30, B=@n32), B=Multiply(A=@n34, B=@n35)), Speed=AppendVector(A=@n31.B, B=@n31.B)) {speedX=0.2, speedY=0.2}
  n37 = TextureSampleParameter2D(UVs=@n36) {parameterName=MaskTex, samplerType=SAMPLERTYPE_LinearColor, texture=&MergedFireFlowMap}
  n38 = TextureCoordinate()
  n39 = Multiply(A=Abs(None=Subtract(A=@n38) {constB=0.5})) {constB=5}
  n40 = Add(
   A=Add(
    A=TextureSampleParameter2D(UVs=@n36) {parameterName=MaskTex02, texture=/BRCosmetics/Weapons/FORT_Melee/Pickaxe_SoapPocket/FX/T_Pickaxe_SoapPocket_Noise}.R
    B=Subtract(A=@n37.R) {constB=0.5}
   )
   B=ComponentMask(None=OneMinus(None=@n39)) {r=true}
  )
  n41 = ScalarParameter() {defaultValue=1, parameterName=MaskMul}
  Saturate(None=Power(Base=Multiply(A=@n40, B=@n41), Exp=ScalarParameter() {defaultValue=4, parameterName=Mask_Power}) {constExponent=4})
 ]
}
