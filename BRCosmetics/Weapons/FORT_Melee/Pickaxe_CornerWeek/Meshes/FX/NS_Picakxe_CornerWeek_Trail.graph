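# Niagara trail effect for the CornerWeek pickaxe, built from five ribbon emitters that use anim trail modules between pairs of bone or socket indices.
# Center, In and Out ribbons break up with panned curl noise, narrow over life and drive dissolve-style dynamic material parameters from curves.
# Two broader layers scale color and alpha over life with time-of-day scalars, one tinted orange and the other given randomized material parameters.
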NiagaraSystem /BRCosmetics/Weapons/FORT_Melee/Pickaxe_CornerWeek/Meshes/FX/NS_Picakxe_CornerWeek_Trail
dependencies {
 usedDataInterfaces [&NiagaraDataInterfaceCurve, &NiagaraDataInterfaceColorCurve, &NiagaraDataInterfaceSocketReader]
 usedDynamicInputs []
 usedModules [
  &self:SystemSpawnScript.SystemScriptSource.SystemScriptGraph.NiagaraNodeAssignment_0.SetVariables_#1
  &self:SystemSpawnScript.SystemScriptSource.SystemScriptGraph.NiagaraNodeAssignment_1.SetVariables_#2
  &SystemState
 ]
}
emitters {
 Trail001 {
  emitterUpdateScript [
   &EmitterState(Life Cycle Mode=::NewEnumerator1, Loop Behavior=::NewEnumerator1, Loop Duration=0.5, UseLoopDelay=false)
   &FN_SpawnAnimTrail()
  ]
  particleSpawnScript [
   &FN_InitializeAnimTrailBoneIndecies(First Bone or Socket Index=1, Second Bone or Socket Index=3)
   &FN_InitializeAnimTrail(Lifetime=0.18)
  ]
  particleUpdateScript [
   &FN_UpdateAnimTrail()
   &DynamicMaterialParameters(
    Index 0 Param 1=&UniformRangedFloat(Minimum=0, Maximum=1)
    Index 0 Param 2=&UniformRangedFloat(Minimum=0, Maximum=1)
    Index 0 Param 3=1
    Index 0 Param 4=1
    Module.Param0WriteEnabled=true
    Module.Param1WriteEnabled=true
    Module.Param2WriteEnabled=true
    Module.Param3WriteEnabled=true
   )
   &ScaleColor(
    Scale Alpha=&FloatFromCurve(
     FloatCurve=DataInterface() {propertyValues {Curve {keys [
         {arriveTangent=-2.5493, interpMode=RCIM_Cubic, leaveTangent=-2.5493, tangentMode=RCTM_User, time=0, value=1}
         {arriveTangent=-0.0154, interpMode=RCIM_Cubic, leaveTangent=-0.0154, tangentMode=RCTM_User, time=0.6, value=0}
        ]}}}
     Scale Curve=&FNTimeOfDayScalar(Night Scale=1, Morning Scale=0.7, Evening Scale=1)
    )
    Scale RGB=&Multiply_Vector(
     A=&VectorFromFloat(Value=&FNTimeOfDayScalar(Night Scale=0.3, Morning Scale=0.3, Evening Scale=0.3))
     B=&MakeVectorFromLinearColorRGB(Color=&ColorFromCurve(ColorCurve=DataInterface() {propertyValues {
         AlphaCurve {keys [<at 0 1 RCIM_Cubic>]}
         BlueCurve {keys [<at 0 1 RCIM_Cubic>,<at 0.9913 1>]}
         GreenCurve {keys [<at 0 1 RCIM_Cubic>,<at 0.9913 1>]}
         RedCurve {keys [<at 0 1 RCIM_Cubic>,<at 0.9913 1>]}
        }}))
    )
    ScaleA=true
    ScaleRGB=true
   )
  ]
  rendererClasses [&NiagaraRibbonRendererProperties]
  renderers [{rendererClass=&NiagaraRibbonRendererProperties}]
 }
 Trail004 {
  emitterUpdateScript [
   &EmitterState(Life Cycle Mode=::NewEnumerator1, Loop Behavior=::NewEnumerator1, Loop Duration=0.5, UseLoopDelay=false)
   &FN_SpawnAnimTrail()
  ]
  particleSpawnScript [
   &FN_InitializeAnimTrailBoneIndecies(First Bone or Socket Index=1, Second Bone or Socket Index=3)
   &FN_InitializeAnimTrail(AnimTrailWidth=1.2, Lifetime=0.2)
  ]
  particleUpdateScript [
   &FN_UpdateAnimTrail()
   &ScaleColor(
    Scale Alpha=&FloatFromCurve(
     FloatCurve=DataInterface() {propertyValues {Curve {keys [
         {arriveTangent=-1.3664, interpMode=RCIM_Cubic, leaveTangent=-1.3664, tangentMode=RCTM_User, time=0, value=1}
         {arriveTangent=-1.0629, interpMode=RCIM_Cubic, leaveTangent=-1.0629, tangentMode=RCTM_User, time=0.4, value=0.4}
         <at 1 0 RCIM_Cubic>
        ]}}}
     Scale Curve=&Multiply_Float(A=&FNTimeOfDayScalar(Night Scale=0.6, Morning Scale=0.8, Evening Scale=0.7), B=1)
    )
    Scale RGB=&MakeVectorFromLinearColorRGB(Color=<rgba 0.8229 0.3125 0 1>)
    ScaleA=true
    ScaleRGB=true
   ) {module=ScaleColor002}
  ]
  rendererClasses [&NiagaraRibbonRendererProperties]
  renderers [{rendererClass=&NiagaraRibbonRendererProperties}]
 }
 Trail_Center {
  emitterUpdateScript [
   &EmitterState(Life Cycle Mode=::NewEnumerator1, Loop Behavior=::NewEnumerator1, Loop Duration=0.5, UseLoopDelay=false)
   &FN_SpawnAnimTrail()
  ]
  particleSpawnScript [
   &FN_InitializeAnimTrailBoneIndecies(First Bone or Socket Index=4, Second Bone or Socket Index=5)
   &FN_InitializeAnimTrail(AnimTrailWidth=1.3, Lifetime=0.26)
  ]
  particleUpdateScript [
   &FN_UpdateAnimTrail()
   /SetVariables(Particles.RibbonFacing=<1 0 0>) {enabled=false, isSetParameters=true}
   &CurlNoiseForce(Bias Noise=false, Noise Frequency=30, Noise Strength=1000, Pan Noise=true, Pan Noise Field=<0.1 0.2 0.3>)
   &Drag(Drag=1.5, Use Linear Drag=true, Use Rotational Drag=true) {enabled=false}
   &SolveForcesAndVelocity(Clamp Velocity=false, Limit Acceleration=false)
   &DynamicMaterialParameters(
    Index 0 Param 1=&FloatFromCurve(FloatCurve=DataInterface() {propertyValues {Curve {keys [<at 0 1 RCIM_Cubic>, {arriveTangent=-2, interpMode=RCIM_Cubic, leaveTangent=-2, time=0.665, value=0.6902}, <at 1 -1 RCIM_Cubic>]}}})
    Index 0 Param 2=$Particles.MaterialRandom
    Index 0 Param 3=0.6
    Index 0 Param 4=1
    Module.Param0WriteEnabled=true
    Module.Param1WriteEnabled=true
    Module.Param2WriteEnabled=true
    Module.Param3WriteEnabled=true
   )
   &ScaleRibbonWidth(Ribbon Width Scale=&FloatFromCurve(FloatCurve=DataInterface() {propertyValues {Curve {keys [
         {arriveTangent=-0.1971, interpMode=RCIM_Cubic, leaveTangent=-0.1971, tangentMode=RCTM_User, time=0, value=1}
         {arriveTangent=-2.7059, interpMode=RCIM_Cubic, leaveTangent=-2.7059, tangentMode=RCTM_User, time=0.3268, value=0.3962}
         {arriveTangent=-0.8188, interpMode=RCIM_Cubic, leaveTangent=-0.8188, tangentMode=RCTM_User, time=0.5, value=0}
        ]}}}))
  ]
  rendererClasses [&NiagaraRibbonRendererProperties]
  renderers [{rendererClass=&NiagaraRibbonRendererProperties}]
 }
 Trail_In {
  emitterUpdateScript [
   &EmitterState(Life Cycle Mode=::NewEnumerator1, Loop Behavior=::NewEnumerator1, Loop Duration=0.5, UseLoopDelay=false)
   &FN_SpawnAnimTrail()
  ]
  particleSpawnScript [
   &FN_InitializeAnimTrailBoneIndecies()
   &FN_InitializeAnimTrail(AnimTrailWidth=0.6, Lifetime=0.26)
  ]
  particleUpdateScript [
   &FN_UpdateAnimTrail()
   /SetVariables(Particles.RibbonFacing=<1 0 0>) {enabled=false, isSetParameters=true}
   &CurlNoiseForce(Bias Noise=false, Noise Frequency=20, Noise Strength=1000, Pan Noise=true, Pan Noise Field=<0.1 0.1 0.1>)
   &Drag(Drag=1.5, Use Linear Drag=true, Use Rotational Drag=true) {enabled=false}
   &SolveForcesAndVelocity(Clamp Velocity=false, Limit Acceleration=false)
   &DynamicMaterialParameters(
    Index 0 Param 1=&FloatFromCurve(FloatCurve=DataInterface() {propertyValues {Curve {keys [<at 0 1 RCIM_Cubic>, {arriveTangent=-2, interpMode=RCIM_Cubic, leaveTangent=-2, time=0.665, value=0.6902}, <at 1 -1 RCIM_Cubic>]}}})
    Index 0 Param 2=$Particles.MaterialRandom
    Index 0 Param 3=0.6
    Index 0 Param 4=1
    Module.Param0WriteEnabled=true
    Module.Param1WriteEnabled=true
    Module.Param2WriteEnabled=true
    Module.Param3WriteEnabled=true
   )
   &ScaleRibbonWidth(Ribbon Width Scale=&FloatFromCurve(FloatCurve=DataInterface() {propertyValues {Curve {keys [
         <at 0 1 RCIM_Cubic>
         {arriveTangent=-2.5361, interpMode=RCIM_Cubic, leaveTangent=-2.5361, tangentMode=RCTM_User, time=0.2799, value=0.3445}
         {arriveTangent=-0.2954, interpMode=RCIM_Cubic, leaveTangent=-0.2954, tangentMode=RCTM_User, time=0.5, value=0}
        ]}}}))
  ]
  rendererClasses [&NiagaraRibbonRendererProperties]
  renderers [{rendererClass=&NiagaraRibbonRendererProperties}]
 }
 Trail_Out {
  emitterUpdateScript [
   &EmitterState(Life Cycle Mode=::NewEnumerator1, Loop Behavior=::NewEnumerator1, Loop Duration=0.5, UseLoopDelay=false)
   &FN_SpawnAnimTrail()
  ]
  particleSpawnScript [
   &FN_InitializeAnimTrailBoneIndecies(First Bone or Socket Index=2, Second Bone or Socket Index=3)
   &FN_InitializeAnimTrail(AnimTrailWidth=0.8, Lifetime=0.26)
  ]
  particleUpdateScript [
   &FN_UpdateAnimTrail()
   /SetVariables(Particles.RibbonFacing=<1 0 0>) {isSetParameters=true}
   &CurlNoiseForce(Bias Noise=true, Bias Noise Field:Vector3f=<0.1 0 0>, Noise Frequency=20, Noise Strength=1000, Pan Noise=true, Pan Noise Field=<0.1 0.1 0.1>)
   &Drag(Drag=1.5, Use Linear Drag=true, Use Rotational Drag=true) {enabled=false}
   &SolveForcesAndVelocity(Clamp Velocity=false, Limit Acceleration=false)
   &DynamicMaterialParameters(
    Index 0 Param 1=&FloatFromCurve(FloatCurve=DataInterface() {propertyValues {Curve {keys [<at 0 1 RCIM_Cubic>, {arriveTangent=-2, interpMode=RCIM_Cubic, leaveTangent=-2, time=0.665, value=0.6902}, <at 1 -1 RCIM_Cubic>]}}})
    Index 0 Param 2=$Particles.MaterialRandom
    Index 0 Param 3=0.6
    Index 0 Param 4=1
    Module.Param0WriteEnabled=true
    Module.Param1WriteEnabled=true
    Module.Param2WriteEnabled=true
    Module.Param3WriteEnabled=true
   )
   &ScaleRibbonWidth(Ribbon Width Scale=&FloatFromCurve(FloatCurve=DataInterface() {propertyValues {Curve {keys [
         <at 0 1 RCIM_Cubic>
         {arriveTangent=-3.1902, interpMode=RCIM_Cubic, leaveTangent=-3.1902, tangentMode=RCTM_User, time=0.2785, value=0.3441}
         {arriveTangent=-0.2954, interpMode=RCIM_Cubic, leaveTangent=-0.2954, tangentMode=RCTM_User, time=0.5, value=0}
        ]}}}))
  ]
  rendererClasses [&NiagaraRibbonRendererProperties]
  renderers [{rendererClass=&NiagaraRibbonRendererProperties}]
 }
}
summary {
 systemName=NS_Picakxe_CornerWeek_Trail
 emitters [
  {emitterName=Trail_In, rendererClasses [&NiagaraRibbonRendererProperties]}
  {emitterName=Trail_Out, rendererClasses [&NiagaraRibbonRendererProperties]}
  {emitterName=Trail_Center, rendererClasses [&NiagaraRibbonRendererProperties]}
  {emitterName=Trail004, rendererClasses [&NiagaraRibbonRendererProperties]}
  {emitterName=Trail001, rendererClasses [&NiagaraRibbonRendererProperties]}
 ]
 userVariables [
  User.bIsFrontEnd {type=NiagaraBool}
  User.bIsFrontEndPreview {type=NiagaraBool}
  User.PositionLerp {type=NiagaraFloat, defaultValue=0, description=""}
  User.RibbonIDOffset {type=NiagaraInt32}
  User.SkeletalMesh {
   type=NiagaraDataInterfaceSkeletalMesh
   defaultValue {
    struct=NiagaraExt_VariableValue_DataInterface
    value {dataInterface=&self:NiagaraDataInterfaceSkeletalMesh_0}
   }
  }
 ]
}
